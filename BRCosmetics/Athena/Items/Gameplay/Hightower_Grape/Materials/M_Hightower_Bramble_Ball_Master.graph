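# Master material for the Hightower Grape bramble ball shield, built for Niagara mesh particles, with a static switch choosing inner shell or outer look.
# The inner branch uses shield textures and scales vertices about the object pivot with a Scale vector.
# The other branch reuses Grape glider textures, feeds a particle-alpha-driven noise dissolve into opacity mask, and pulls vertices toward the pivot by vertex color dotted with a Mask vector.
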Material /BRCosmetics/Athena/Items/Gameplay/Hightower_Grape/Materials/M_Hightower_Bramble_Ball_Master
graph {
 properties {
  bUseMaterialAttributes=true
  bUsedWithNiagaraMeshParticles=true
  pixelDepthOffsetMode=PDOM_Legacy
  previewMesh=/BRCosmetics/Athena/Items/Gameplay/Hightower_Grape/Mesh/sm_hightower_grape_shield_meshes
  usedShadingModels=Default Lit
 }
 nodes [
  Constant() {r=1}
  Constant() {r=150}
  n1 = TextureSample() {samplerType=SAMPLERTYPE_Normal, texture=/BRCosmetics/Athena/Items/Gameplay/Hightower_Grape/Textures/T_Hightower_Grape_Shield_Inner_N}
  n2 = ObjectPositionWS()
  n3 = WorldPosition()
  n4 = SetMaterialAttributes!(
   MaterialAttributes
   Base Color=TextureSample() {texture=/BRCosmetics/Athena/Items/Gameplay/Hightower_Grape/Textures/T_Hightower_Grape_Shield_Inner_D}.RGB
   Normal=@n1.RGB
   Opacity Mask=TextureSample() {texture=/BRCosmetics/Athena/Items/Gameplay/Hightower_Grape/Textures/T_Hightower_Grape_Shield_Inner_S}.G
   Roughness=Constant() {r=0.75}
   World Position Offset=Multiply(A=Subtract(A=@n2, B=@n3.XYZ), B=VectorParameter() {parameterName=Scale, defaultValue=<rgba 0 0 0 1>}.RGB)
  ) {attributeSetTypes [
    #1
    #2
    #3
    #4
    #5
   ]}
  n5 = TextureSample() {
   samplerType=SAMPLERTYPE_Normal
   texture=/BRCosmetics/Gadgets/Assets/VinderTech_GliderChute/Glider_Hightower_Grape/Textures/T_HighTower_Grape_N
  }
  n6 = Multiply(A=TextureCoordinate(), B=ScalarParameter() {defaultValue=0.5, parameterName=Dissolve Multiplier Size})
  n7 = Subtract(
   A=Add(A=TextureSample(UVs=@n6) {samplerType=SAMPLERTYPE_LinearColor, texture=/Game/Global/Textures/Default/Noise/Noise32x32}.R)
   B=LinearInterpolate(Alpha=ParticleColor().A) {constA=1.2, constB=0}
  )
  n8 = TextureSample() {
   samplerType=SAMPLERTYPE_Masks
   texture=/BRCosmetics/Gadgets/Assets/VinderTech_GliderChute/Glider_Hightower_Grape/Textures/T_HighTower_Grape_S
  }
  n9 = ObjectPositionWS()
  n10 = WorldPosition()
  n11 = VertexColor()
  n12 = VectorParameter() {parameterName=Mask, defaultValue=<rgba 0 0 0 1>}
  n13 = SetMaterialAttributes(
   Base Color=TextureSample() {texture=/BRCosmetics/Gadgets/Assets/VinderTech_GliderChute/Glider_Hightower_Grape/Textures/T_HighTower_Grape_D}.RGB
   Normal=@n5.RGB
   Opacity Mask=Clamp(None=Floor(None=@n7))
   Roughness=@n8.B
   World Position Offset=Multiply(A=Subtract(A=@n9, B=@n10.XYZ), B=DotProduct(A=@n11, B=@n12.RGB))
  ) {attributeSetTypes [
    #1
    #2
    #3
    #4
    #5
   ]}
  StaticSwitchParameter(True=@n4, False=@n13) {parameterName=IsInnerShell?}
 ]
}
